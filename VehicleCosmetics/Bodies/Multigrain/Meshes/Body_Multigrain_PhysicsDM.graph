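# Physics asset for the Multigrain vehicle body cosmetic mesh, providing a simplified collision shape for the car.
# It holds a single body on the root bone with one box shape covering the hull, offset slightly along X and raised above the origin.
# It defines no constraints, so the body simulates as one rigid piece.
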PhysicsAsset /VehicleCosmetics/Bodies/Multigrain/Meshes/Body_Multigrain_PhysicsDM
bodies {root [{shapeType=Box, extentX=444, extentY=216, extentZ=116, shapeName=root_box, center=<22.801 0 95.6529>}]}
constraints []
